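# Niagara effect for the butterfly knife in the SpyMale main-player emote set, showing a short-lived, fading mesh.
# A single Mesh emitter burst-spawns one tinted particle with scale, initial orientation and rotational force set at spawn.
# Update solves forces and rotation, orients the mesh to a vector, and fades alpha to near zero over its life through a curve.
# The system exposes no user parameters.
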NiagaraSystem /BRCosmetics/Animation/Game/MainPlayer/Emotes/SpyMale/Effects/NS_ButterflyKnife
dependencies {usedDataInterfaces [&NiagaraDataInterfaceCurve], usedDynamicInputs [], usedModules [&SystemState]}
emitters {Mesh {
  emitterUpdateScript [
   &EmitterState(Life Cycle Mode=::NewEnumerator1, Loop Behavior=::NewEnumerator1, Loop Duration=0.5, UseLoopDelay=false)
   &SpawnBurst_Instantaneous(Spawn Count=1, Use Spawn Probability=false)
   &SpawnRate(SpawnRate=0, Use Spawn Probability=false)
  ]
  particleSpawnScript [
   /Niagara/Modules/Spawn/Initialization/InitializeParticle(
    Color=<rgba 0.0825 0.1 0.0808 1>
    Lifetime=0.5
    Mesh Scale=<60 60 50>
    Module.Write Color=true
    Module.Write Lifetime=true
    Module.Write Mass=true
    Module.Write Material Random=true
    Module.Write Position=false
    Module.Write Scale=true
    Module.Write SpriteRotation=false
    Module.Write SpriteSize=false
    Sprite Rotation=&UniformRangedFloat(Minimum=0, Maximum=360)
    Sprite Size=<70 70>
   )
   /Niagara/Modules/Spawn/Orientation/V2/InitialMeshOrientation(
    Orientation Coordinate Space:ENiagaraCoordinateSpace=::Local
    Orientation Vector=<0 0 1>
    Rotation=&UniformRangedVector(Minimum=<0 0 0>, Maximum=<0 0 1>)
    Use Orientation Vector=true
    Use Rotation Vector=true
   )
   &MeshRotationForce(Coordinate Space:ENiagaraMeshTransforms=::NewEnumerator3, Rotation=<0 0 360>) {module=MeshRotationForce001}
   &ApplyInitialForces(Apply Force to Velocity=false)
   &AddRotationalVelocity()
  ]
  particleUpdateScript [
   &ParticleState() {module=UpdateAge}
   &ScaleColor(
    Scale Alpha=&FloatFromCurve(FloatCurve=DataInterface() {propertyValues {Curve {keys [<at -0.0046 1.0113 RCIM_Cubic>,<at 1.0011 0.002 RCIM_Cubic>]}}})
    ScaleA=true
    ScaleRGB=true
   )
   &SolveForcesAndVelocity(Clamp Velocity=false, Limit Acceleration=false)
   &SolveRotationalForcesAndVelocity()
   /Niagara/Modules/Update/Orientation/V2/OrientMeshToVector(Look At Direction:Vector3f=<1 0 0>, Override Quat=false, UseOptionalSideAxis=false, UseOptionalUpAxis=false)
  ]
  rendererClasses [&NiagaraMeshRendererProperties]
  renderers [{rendererClass=&NiagaraMeshRendererProperties, rendererIndex=0}]
 }}
summary {
 systemName=NS_ButterflyKnife
 emitters [{emitterName=Mesh, rendererClasses [&NiagaraMeshRendererProperties]}]
 userVariables []
}
